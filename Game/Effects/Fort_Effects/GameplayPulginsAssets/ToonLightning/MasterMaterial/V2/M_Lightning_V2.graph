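# Translucent unlit master material for toon lightning bolts on Niagara mesh particles, with a static switch between textured bolt and soft glow looks.
# Position, tangent and thickness texture arrays, indexed per particle, drive WPO that bends the mesh into a camera-facing strip.
# Dynamic parameters control growth, thickness and texture dissolves, center pull and tangent alignment.
# Noise WPO is disabled on low-performance platforms, and thickness data also boosts spike brightness.
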Material /Game/Effects/Fort_Effects/GameplayPulginsAssets/ToonLightning/MasterMaterial/V2/M_Lightning_V2
graph {
 properties {
  bUsedWithNiagaraMeshParticles=true
  blendMode=BLEND_Translucent
  pixelDepthOffsetMode=PDOM_Legacy
  shadingModel=MSM_Unlit
  usedShadingModels=Unlit
 }
 nodes [
  n1 = NamedRerouteUsage()
  n2 = NamedRerouteUsage()
  Add(A=Add(A=NamedRerouteUsage(), B=Add(A=@n1, B=@n2)), B=NamedRerouteUsage())
  DynamicParameter() {parameterIndex=2, defaultValue=<rgba 0 1 1 1>, paramNames [MoveToCenter,TextureWPO,Param3,Param4]}
  DynamicParameter() {parameterIndex=1, defaultValue=<rgba 1 0 0 0>, paramNames [GrowthProgress, ThicknessDissolve, TextureDissolve, MoveAlignOriginalTangent]}
  DynamicParameter() {parameterIndex=3, paramNames [Param1,Param2,Param3,RandomInput]}
  n3 = NamedRerouteUsage()
  n4 = NamedRerouteUsage()
  n5 = Multiply(
   A=NamedRerouteUsage()
   B=Multiply(A=NamedRerouteUsage(), B=StaticSwitchParameter(True=@n3, False=@n4) {group="##01.GlobalSwitcher", parameterName=GlowMode})
  )
  n6 = ParticleColor()
  MaterialFunctionCall(Emissive=Divide(A=Multiply(A=@n5, B=@n6.RGB)) {constB=2.1}) {materialFunction=&MF_FX_EmissiveAdjustments}
  n7 = NamedRerouteUsage()
  n8 = NamedRerouteUsage()
  n9 = Multiply(
   A=NamedRerouteUsage()
   B=Multiply(A=NamedRerouteUsage(), B=StaticSwitchParameter(True=@n7, False=@n8) {group="##01.GlobalSwitcher", parameterName=GlowMode})
  )
  n10 = ParticleColor()
  Multiply(
   A=DepthFade(Opacity=Multiply(A=@n9, B=@n10.A), FadeDistance=ScalarParameter() {defaultValue=1, group=CommonFade, parameterName=DepthFade})
   B=MaterialFunctionCall(Distance=ScalarParameter() {group=CommonFade, parameterName=NearCameraFade}) {materialFunction=&Near_Camera_Fade}.Result
  )
  n11 = SphereMask(
   A=NamedRerouteUsage()
   B=Constant3Vector()
   Radius=ScalarParameter() {defaultValue=0.9, group=MoveAlignOriginalTangent, parameterName=MoveAlignOriginalTangent_Radius}
  ) {attenuationRadius=1, hardnessPercent=0}
  n12 = DynamicParameter() {parameterIndex=1, defaultValue=<rgba 1 0 0 0>, paramNames [GrowthProgress, ThicknessDissolve, TextureDissolve, MoveAlignOriginalTangent]}
  n13 = Multiply(
   A=Power(
    Base=OneMinus(None=@n11)
    Exp=ScalarParameter() {defaultValue=2, group=MoveAlignOriginalTangent, parameterName=MoveAlignOriginalTangent_Power}
   )
   B=@n12.MoveAlignOriginalTangent
  )
  n14 = NamedRerouteUsage()
  n15 = Constant3Vector() {constant=<rgba 0 -1 0 0>}
  NamedRerouteDeclaration(None=Multiply(A=Multiply(A=@n13, B=@n14), B=Transform(None=@n15))) {name=WPO_MoveAlignLightningDir, nodeColor=<rgba 1 0 0.3047 1>}
  n16 = MaterialFunctionCall() {materialFunction=&LocalPosition}
  n17 = NamedRerouteUsage()
  n18 = DynamicParameter() {parameterIndex=2, defaultValue=<rgba 0 1 1 1>, paramNames [MoveToCenter,TextureWPO,Param3,Param4]}
  n19 = Multiply(A=Subtract(A=Constant3Vector(), B=Add(A=@n16.Instance Local Position, B=@n17)), B=@n18.MoveToCenter)
  n20 = Constant3Vector() {constant=<rgba 1 1 0 0>}
  NamedRerouteDeclaration(None=Transform(None=Multiply(A=@n19, B=@n20)) {transformSourceType=TRANSFORMSOURCE_Local}) {name=WPO_MoveToCenter, nodeColor=<rgba 1 0.7987 0 1>}
  n21 = TextureObjectParameter() {
   group=LightningShape
   parameterName=LightningShape_Texture
   sortPriority=0
   texture=&/LightningTexture
  }
  n22 = VectorParameter() {
   group=LightningShape
   parameterName=LightningShape_UVControl
   channelNames {a=VPan, b=UPan, g=VScale, r=UScale}
   defaultValue=<rgba 0.8 1.5 0 0>
  }
  n23 = MaterialFunctionCall!(
   Texture=@n21
   AdditionalUVOffset
   UVControl_RGB=@n22
   UVControl_A=@n22
   "[Override]TimeInput"=NamedRerouteUsage()
   "[Override]ScaleUVByCenter"=StaticBool() {value=true}
   "[Override]UVInput"=NamedRerouteUsage()
  ) {materialFunction=/Game/Effects/Niagara/MasterMaterials/Functions/Common/FXCommon_TextureSample_V2}
  n24 = VectorParameter() {group=Color, parameterName=OutGlowRefine, defaultValue=<rgba 1.5 0.5 0 0>}
  n25 = Saturate(None=Power(Base=@n23.G, Exp=ScalarParameter() {group=Color, parameterName=ShapeEdgeFadePower}))
  n26 = Add(
   A=MaterialFunctionCall(Texture=@n23.R, RGB=@n24.RGB, A=@n24.RGB) {materialFunction=&MF_FX_Refine}.Result
   B=Saturate(None=Multiply(A=@n25, B=@n23.B))
  )
  n27 = Constant2Vector() {r=1}
  NamedRerouteDeclaration(None=Multiply(
    A=Saturate(None=@n26)
    B=MaterialFunctionCall!(TexCoord=@n23.UV, ClampChannel=@n27) {materialFunction=/Game/Effects/Fort_Effects/MaterialFunctions/MF_FX_TilingMethod}.Result
   )) {name=LightningModeAlpha, nodeColor=<rgba 1 0.7734 0 1>}
  n28 = MaterialFunctionCall(Float2=NamedRerouteUsage()) {materialFunction=&BreakOutFloat2Components}
  n29 = VectorParameter() {group=Color, parameterName=GlowSideOpacityMask, channelNames {a="", b=Power, g=Multiply, r=Offset}, defaultValue=<rgba 0 1.5 2 1>}
  NamedRerouteDeclaration(None=Multiply(
    A=MaterialFunctionCall!(UorV=@n28.R, EdgeControl=@n29) {materialFunction=/Game/Effects/Niagara/MasterMaterials/Functions/Common/FXCommon_EdgeMask_Simple}.Mask
    B=ScalarParameter() {defaultValue=1, group=Color, parameterName=GlowOpacityScale}
   )) {name=GlowModeAlpha, nodeColor=<rgba 0.9219 0 1 1>}
  n30 = NamedRerouteUsage()
  n31 = VectorParameter() {group=Color, parameterName=GlowBrightenRefine, defaultValue=<rgba 4 2 0.5 0.5>}
  n32 = MaterialFunctionCall(
   Texture=MaterialFunctionCall(
    UorV=MaterialFunctionCall(Float2=@n30) {materialFunction=&BreakOutFloat2Components}.R
    EdgeControl=Constant3Vector() {constant=<rgba 0 2 1 0>}
   ) {materialFunction=&FXCommon_EdgeMask_Simple}.Mask
   RGB=@n31.RGB
   A=@n31.RGB
  ) {materialFunction=&MF_FX_Refine}
  NamedRerouteDeclaration(None=Multiply(A=VectorParameter() {group=Color, parameterName=GlowColor, defaultValue=<rgba 0 0.8333 5 1>}.RGB, B=Add(A=@n32.Result))) {name=GlowModeColor, nodeColor=<rgba 0 1 0.8828 1>}
  n33 = NamedRerouteUsage()
  n34 = VectorParameter() {group=Color, parameterName=OpacityFade_Tip, channelNames {a="", b=Power, g=Multiply, r=Offset}, defaultValue=<rgba 999 1 1 1>}
  n35 = VectorParameter() {group=Color, parameterName=OpacityFade_Bottom, channelNames {a="", b=Power, g=Multiply, r=Offset}, defaultValue=<rgba 999 1 1 1>}
  NamedRerouteDeclaration(None=MaterialFunctionCall!(
    UorV=MaterialFunctionCall(Float2=@n33) {materialFunction=&BreakOutFloat2Components}.G
    Invert_EdgeControl=@n34
    NoInvert_EdgeControl=@n35
   ) {materialFunction=/Game/Effects/Niagara/MasterMaterials/Functions/Common/FXCommon_EdgeMask_Ex}.Mask) {name=TopAndBottomFade, nodeColor=<rgba 0 0.7187 1 1>}
  n36 = MaterialFunctionCall(Float2=NamedRerouteUsage()) {materialFunction=&BreakOutFloat2Components}
  n37 = VectorParameter() {group=Color, parameterName=TipAndBottomBrighten, defaultValue=<rgba 5 5 1 0>}
  n38 = Multiply(A=Power(Base=@n36.G, Exp=@n37.R), B=@n37.G)
  n39 = Multiply(A=Power(Base=OneMinus(None=@n36.G), Exp=@n37.B), B=@n37.A)
  NamedRerouteDeclaration(None=Add(A=Add(A=@n38, B=@n39))) {name=TipAndBottomBright, nodeColor=<rgba 0.5703 1 0 1>}
  n40 = MaterialFunctionCall(Float2=NamedRerouteUsage()) {materialFunction=&BreakOutFloat2Components}
  n41 = OneMinus(None=@n40.G)
  n42 = ScalarParameter() {group=Growth, parameterName=GrowthDirectionInvert}
  n43 = DynamicParameter() {parameterIndex=1, defaultValue=<rgba 1 0 0 0>, paramNames [GrowthProgress, ThicknessDissolve, TextureDissolve, MoveAlignOriginalTangent]}
  n44 = ScalarParameter() {defaultValue=5, group=Growth, parameterName=GrowthFadeInOutLength}
  n45 = Divide(B=@n44)
  n46 = Add(A=Subtract(A=LinearInterpolate(A=@n40.G, B=@n41, Alpha=@n42)), B=Multiply(A=Saturate(None=@n43.GrowthProgress), B=Add(B=@n45) {constA=1}))
  NamedRerouteDeclaration(None=Saturate(None=Multiply(A=@n46, B=@n44))) {name=LightningGrowthMask, nodeColor=<rgba 0.875 1 0 1>}
  n47 = NamedRerouteUsage()
  n48 = MaterialFunctionCall(
   R=OneMinus(None=MaterialFunctionCall(Float2=@n47) {materialFunction=&BreakOutFloat2Components}.G)
   G=Constant() {r=0.5}
   B=Floor(None=DynamicParameter() {paramNames [MeshScaleX, LightningDataIndex, ThicknessScale, ShapeWPOScale]}.LightningDataIndex)
  ) {materialFunction=&AppendMany}
  n49 = TextureSampleParameter2DArray!(UVs=@n48.RGB, Apply View MipBias) {
   group="##00.LightningData"
   parameterName=LightningTexture_PositionArray
   samplerType=SAMPLERTYPE_LinearColor
   texture=/Game/Effects/Fort_Effects/GameplayPulginsAssets/ToonLightning/GP29/Texture/T_GP29Lightning_PositionArray
  }
  NamedRerouteDeclaration(None=@n49.RGB) {name=LightningData_Position, nodeColor=<rgba 0.2422 0 1 1>}
  n50 = TextureSampleParameter2DArray(UVs=@n48.RGB) {
   group="##00.LightningData"
   parameterName=LightningTexture_TangentArray
   samplerType=SAMPLERTYPE_LinearColor
   texture=/Game/Effects/Fort_Effects/GameplayPulginsAssets/ToonLightning/GP29/Texture/T_GP29Lightning_TangentArray
  }
  NamedRerouteDeclaration(None=@n50.RGB) {name=LightningData_Tangent, nodeColor=<rgba 1 0.3584 0 1>}
  n51 = NamedRerouteUsage()
  n52 = MaterialFunctionCall(Float3=Multiply(A=Subtract(A=@n51) {constB=0.5}) {constB=2}) {materialFunction=&BreakOutFloat3Components}
  n53 = ScalarParameter() {group="##00.LightningData", parameterName=InvertLightningPositionAndTangent}
  n54 = LinearInterpolate(Alpha=Saturate(None=Floor(None=@n53))) {constA=1, constB=-1}
  n55 = MaterialFunctionCall(X=Multiply(A=@n52.R, B=@n54), Y=Multiply(A=@n52.G, B=@n54), Z=@n52.B) {materialFunction=&MakeFloat3}
  n56 = Multiply(
   A=@n55.Result
   B=Multiply(
    A=ScalarParameter() {defaultValue=35, group="##00.LightningData", parameterName=PositionDataScale}
    B=DynamicParameter() {paramNames [MeshScaleX, LightningDataIndex, ThicknessScale, ShapeWPOScale]}.ShapeWPOScale
   )
  )
  n57 = Transform(None=@n56) {transformSourceType=TRANSFORMSOURCE_Instance}
  n58 = NamedRerouteUsage()
  n59 = Multiply(
   A=MaterialFunctionCall(X=@n54, Y=@n54, Z=Constant() {r=1}) {materialFunction=&MakeFloat3}.Result
   B=Multiply(A=Subtract(A=@n58) {constB=0.5}) {constB=2}
  )
  n60 = CameraPositionWS()
  n61 = Add(A=WorldPosition().XYZ, B=@n57)
  n62 = Normalize(VectorInput=CrossProduct(A=Transform(None=@n59) {transformSourceType=TRANSFORMSOURCE_Instance}, B=Subtract(A=@n60, B=@n61)))
  n63 = NamedRerouteUsage()
  n64 = ScalarParameter() {defaultValue=10, group=SplineWidth, parameterName=BaseWidth}
  n65 = ScalarParameter() {defaultValue=100, group=SplineWidth, parameterName=SpikeWidth}
  n66 = VectorParameter() {group=SplineWidth, parameterName=Spike_RemapThicknessValueRange, defaultValue=<rgba 0 1 0 1>}
  n67 = Power(
   Base=MaterialFunctionCall(Input=NamedRerouteUsage(), Input Low=@n66.R, Input High=@n66.R, Target Low=@n66.R, Target High=@n66.R) {materialFunction=&RemapValueRange}.Result
   Exp=ScalarParameter() {defaultValue=1, group=SplineWidth, parameterName=Spike_ThicknessDataPower}
  )
  n68 = Multiply(
   A=Multiply(
    A=DynamicParameter() {paramNames [MeshScaleX, LightningDataIndex, ThicknessScale, ShapeWPOScale]}.ThicknessScale
    B=LinearInterpolate(A=@n64, B=@n65, Alpha=@n67)
   )
   B=ScalarParameter() {defaultValue=1, group=SplineWidth, parameterName=ScaleSplineWidth}
  )
  n69 = NamedRerouteUsage()
  n70 = Multiply(A=NamedRerouteUsage(), B=Multiply(A=NamedRerouteUsage(), B=Multiply(A=@n68, B=@n69)))
  n71 = MaterialFunctionCall(Float2=NamedRerouteUsage()) {materialFunction=&BreakOutFloat2Components}
  n72 = VectorParameter() {group=SplineWidth, parameterName=SplineWidthMask_Tip, channelNames {a="", b=Power, g=Multiply, r=Offset}, defaultValue=<rgba 5 10 1 1>}
  n73 = VectorParameter() {group=SplineWidth, parameterName=SplineWidthMask_Bottom, channelNames {a="", b=Power, g=Multiply, r=Offset}, defaultValue=<rgba 5 10 1 1>}
  n74 = Constant() {r=-0.5}
  n75 = TextureCoordinate()
  n76 = Multiply(
   A=Multiply(
    A=Multiply(A=@n63, B=@n70)
    B=MaterialFunctionCall(UorV=@n71.G, Invert_EdgeControl=@n72, NoInvert_EdgeControl=@n73) {materialFunction=&FXCommon_EdgeMask_Ex}.Mask
   )
   B=ComponentMask(None=Add(A=@n74, B=@n75)) {r=true}
  )
  NamedRerouteDeclaration(None=Add(A=@n57, B=Multiply(A=@n62, B=@n76))) {name=WPO_PositionOffsetAndThickness, nodeColor=<rgba 0.4766 1 0 1>}
  n77 = DynamicParameter() {parameterIndex=1, defaultValue=<rgba 1 0 0 0>, paramNames [GrowthProgress, ThicknessDissolve, TextureDissolve, MoveAlignOriginalTangent]}
  n78 = Subtract(A=NamedRerouteUsage(), B=@n77.ThicknessDissolve)
  n79 = ScalarParameter() {defaultValue=0.9, group=ThicknessDissolve, parameterName=ThicknessDissolveSoftness}
  NamedRerouteDeclaration(None=Saturate(None=MaterialFunctionCall!(In=@n78, Ramp Width=@n79) {materialFunction=/Engine/Functions/Engine_MaterialFunctions01/Math/SmoothCeil}.Result)) {name=ThicknessDissolve, nodeColor=<rgba 0.2187 1 0 1>}
  NamedRerouteDeclaration(None=@n55.Result) {name=LightningData_OriginalPositionData, nodeColor=<rgba 0.6172 1 0 1>}
  NamedRerouteDeclaration(None=DynamicParameter() {paramNames [MeshScaleX, LightningDataIndex, ThicknessScale, ShapeWPOScale]}.MeshScaleX) {name=MeshScaleXInput, nodeColor=<rgba 0 1 0.6484 1>}
  NamedRerouteDeclaration(None=Add(
    A=Time() {bOverride_Period=true, period=6666}
    B=DynamicParameter() {parameterIndex=3, paramNames [Param1,Param2,Param3,RandomInput]}.RandomInput
   )) {name=TimeWithRandomOffset, nodeColor=<rgba 1 0 0.8672 1>}
  n80 = TextureObjectParameter() {
   group=TextureDissolve
   parameterName=DissolveTexture_Texture
   samplerType=SAMPLERTYPE_LinearColor
   sortPriority=0
   texture=/Engine/Functions/Engine_MaterialFunctions02/ExampleContent/Textures/OffsetNoiseDistanceFields_ForNormals
  }
  n81 = VectorParameter() {
   group=TextureDissolve
   parameterName=DissolveTexture_UVControl
   channelNames {a=VPan, b=UPan, g=VScale, r=UScale}
   defaultValue=<rgba 0.01 2 0 0>
  }
  n82 = MaterialFunctionCall(Texture=@n80, UVControl_RGB=@n81, UVControl_A=@n81, "[Override]UVInput"=NamedRerouteUsage()) {materialFunction=&FXCommon_TextureSample}
  n83 = DynamicParameter() {parameterIndex=1, defaultValue=<rgba 1 0 0 0>, paramNames [GrowthProgress, ThicknessDissolve, TextureDissolve, MoveAlignOriginalTangent]}
  NamedRerouteDeclaration(None=MaterialFunctionCall(
    In=Subtract(A=@n82.R, B=@n83.TextureDissolve)
    Ramp Width=ScalarParameter() {defaultValue=0.5, group=TextureDissolve, parameterName=Dissolve_Softness}
   ) {materialFunction=&SmoothCeil}.Result) {name=Dissolve, nodeColor=<rgba 0.2422 0 1 1>}
  n84 = TextureObjectParameter() {
   group=TextureWPONoise
   parameterName=TextureWPONoise_Texture
   sortPriority=0
   texture=/Game/Effects/Fort_Effects/Emotes/UltraLight/Texture/T_SoftNoise_BC7
  }
  n85 = VectorParameter() {
   group=TextureWPONoise
   parameterName=TextureWPONoise_UVControl
   channelNames {a=VPan, b=UPan, g=VScale, r=UScale}
   defaultValue=<rgba 0.01 3 1 2>
  }
  n86 = Add(
   A=MaterialFunctionCall(
    Texture=@n84
    UVControl_RGB=@n85
    UVControl_A=@n85
    "[Override]TimeInput"=Time() {bOverride_Period=true, period=6666}
    "[Override]UVInput"=NamedRerouteUsage()
   ) {materialFunction=&FXCommon_TextureSample}.RGB
   B=ScalarParameter() {defaultValue=-0.5, group=TextureWPONoise, parameterName=TextureWPONoise_Offset}
  )
  n87 = DynamicParameter() {parameterIndex=2, defaultValue=<rgba 0 1 1 1>, paramNames [MoveToCenter,TextureWPO,Param3,Param4]}
  n88 = Multiply(A=ScalarParameter() {defaultValue=2, group=TextureWPONoise, parameterName=TextureWPONoise_BaseNoiseIntensity}, B=@n87.TextureWPO)
  n89 = MaterialFunctionCall(Float2=NamedRerouteUsage()) {materialFunction=&BreakOutFloat2Components}
  n90 = VectorParameter() {
   group=TextureWPONoise
   parameterName=TextureWPONoise_EdgeMask_InvertControl
   channelNames {a="", b=Power, g=Multiply, r=Offset}
   defaultValue=<rgba 0 10 3 1>
  }
  n91 = VectorParameter() {
   group=TextureWPONoise
   parameterName=TextureWPONoise_EdgeMask_Control
   channelNames {a="", b=Power, g=Multiply, r=Offset}
   defaultValue=<rgba 999 1 1 1>
  }
  n92 = Transform(None=Multiply(
    A=Multiply(A=@n86, B=@n88)
    B=MaterialFunctionCall(UorV=@n89.G, Invert_EdgeControl=@n90, NoInvert_EdgeControl=@n91) {materialFunction=&FXCommon_EdgeMask_Ex}.Mask
   )) {transformSourceType=TRANSFORMSOURCE_Local}
  n93 = Constant3Vector()
  NamedRerouteDeclaration(None=MaterialFunctionCall!("Xb1,PS4"=@n92, "PC-High&&Epic,XSX,PS5"=@n92, "PC-Perf,Android"=@n93, Switch=@n93, "PC-Low(NotPerf)&&Medium"=@n92) {materialFunction=/Game/Effects/Niagara/MasterMaterials/Functions/Common/PlatformSwitch_3D}.Result) {name=TextureWPO, nodeColor=<rgba 1 0.7031 0 1>}
  n94 = TextureCoordinate()
  NamedRerouteDeclaration(None=StaticSwitchParameter(True=MaterialFunctionCall(XY=@n94) {materialFunction=&Swizzle}.YX, False=@n94) {group="##01.GlobalSwitcher", parameterName=SwizzleUV}) {name=TUV, nodeColor=<rgba 0.8281 1 0 1>}
  n95 = TextureSampleParameter2DArray(UVs=@n48.RGB) {
   group="##00.LightningData"
   parameterName=LightningTexture_ThicknessArray
   samplerType=SAMPLERTYPE_LinearColor
   texture=/Game/Effects/Fort_Effects/GameplayPulginsAssets/ToonLightning/GP29/Texture/T_GP29Lightning_ThicknessArray
  }
  NamedRerouteDeclaration(None=@n95.R) {name=LightningData_Thickness, nodeColor=<rgba 1 0.9141 0 1>}
  NamedRerouteDeclaration(None=@n56) {name=LightningData_LocalSpacePosition, nodeColor=<rgba 1 0.375 0 1>}
  n96 = VectorParameter() {group=Color, parameterName=SpikeBrightness_RemapThicknessValueRange, defaultValue=<rgba 0 1 0 1>}
  n97 = Power(
   Base=MaterialFunctionCall(Input=NamedRerouteUsage(), Input Low=@n96.R, Input High=@n96.R, Target Low=@n96.R, Target High=@n96.R) {materialFunction=&RemapValueRange}.Result
   Exp=ScalarParameter() {defaultValue=2, group=Color, parameterName=SpikeBrightness_ThickValuePower}
  )
  n98 = ScalarParameter() {defaultValue=4, group=Color, parameterName=SpikeBrightness_IncreaseColor}
  NamedRerouteDeclaration(None=Add(A=Multiply(A=@n97, B=@n98))) {name=SpikeBrightness, nodeColor=<rgba 1 0 0.2578 1>}
  n99 = VectorParameter() {group=Color, parameterName=BoltColor, defaultValue=<rgba 0 0.8333 1.25 1>}
  n100 = VectorParameter() {group=Color, parameterName=InnerGlowColor, defaultValue=<rgba 0.65 4.7667 13 5>}
  n101 = OneMinus(None=@n23.G)
  n102 = VectorParameter() {group=Color, parameterName=InnerGlowRefine, defaultValue=<rgba 5 2 1 0.5>}
  n103 = Saturate(None=Multiply(
    A=MaterialFunctionCall(Texture=@n101, RGB=@n102.RGB, A=@n102.RGB) {materialFunction=&MF_FX_Refine}.Result
    B=@n23.B
   ))
  NamedRerouteDeclaration(None=LinearInterpolate(
    A=VectorParameter() {group=Color, parameterName=GlowColor, defaultValue=<rgba 0 0.8333 5 1>}.RGB
    B=LinearInterpolate(A=@n99.RGB, B=@n100.RGB, Alpha=@n103)
    Alpha=@n23.B
   )) {name=LightningModeColor, nodeColor=<rgba 0.5 1 0 1>}
  Saturate(None=@n23.R) {desc=R}
  Saturate(None=@n23.B) {desc=B}
  Saturate(None=@n23.G) {desc=G}
 ]
}
